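; Passive Pool Party augment ability that regenerates the player's health and shields while swimming.
; Water-tag add and remove waits start and end the augment effect, re-arming after each exit.
; A water jump defers the end by a one-second timer that rechecks the tag.
; ApplyHealing and ApplyShields overrides layer looping gameplay cues over the parent's regen logic, removing them at upper limits or when the effect ends.
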(event Ability|EventOnEndAbility
  (params (bWasCancelled :type Boolean))
  (CallFunction|RemoveGC false false true)
  (CallFunction|EndAugmentEffect))

(event Ability|EventActivateAbility
  ;[0.1171 1 0] Pasive, wait for swimming to begin
  (CallFunction|WaitforSwimming))

(event Custom|WaitForSwimming
  ;[1 0.2718 0.06097] Activate when swimming
  (Ability|Tasks|WaitGameplayTagAdd (Variables|Default|GetWaterTag) 0 true
    (:Added
      (if true
        (CallFunction|ActivateAugment)
        (else
          _))
      (CallFunction|WaitforStopSwimming))))

(event Custom|WaitForStopSwimming
  ;[1 0 0.933] Wait for the player to begin swimming
  (Ability|Tasks|WaitGameplayTagRemove (Variables|Default|GetWaterTag) 0 true
    (:Removed
      (bind _outfortplayerpawn (CallFunction|GetAvatarasValidFortPlayerPawn)
        (:ValidFortPlayerPawn
          (if (Variables|/Script/FortniteGame.FortPlayerPawn|GetbIsWaterJump _outfortplayerpawn)
            (Utilities|Time|SetTimerbyEvent (EventDispatchers|CreateEvent :!event Custom|CheckForInWater) 1.0)
            (else
              (CallFunction|RemoveGC false false true)
              (CallFunction|EndAugmentEffect))))))))

(event Custom|CheckForInWater
  ;[0 0.2854 1] Check if the player is back in the water after jumping out
  (bind _outfortplayerpawn (CallFunction|GetAvatarasValidFortPlayerPawn)
    (:ValidFortPlayerPawn
      (if (GameplayTags|HasMatchingGameplayTag _outfortplayerpawn (Variables|Default|GetWaterTag))
        (CallFunction|WaitforStopSwimming)
        (else
          (CallFunction|RemoveGC false false true)
          (CallFunction|EndAugmentEffect))))))

(event EventEndAugmentEffect
  ;[1 0.5539 0] Setup the wait again when we leave water
  (|CallParentFunction)
  (CallFunction|WaitforSwimming))

(fn ApplyHealing
  (bind _returnvalue (Ability|GameplayCue|IsGameplayCueActive (Ability|GetAvatarActorfromActorInfo) (Variables|Default|GetGameplayCue_Health)))
  (if _returnvalue
    (|CallParentFunction)
    (if (>= (FortPawn|GetHealth (CallFunction|GetAvatarasFortPlayerPawn)) (ScalableFloat|GetValueatLevel (Variables|Settings|GetHealthUpperLimit)))
      (CallFunction|RemoveGC true))
    (else
      (GameplayCue|AddGameplayCueOnActor(Looping) (Ability|GetAvatarActorfromActorInfo) (Variables|Default|GetGameplayCue_Health) (Ability|GameplayCue|MakeGameplayCueParameters 0.0 (ScalableFloat|GetValueatLevel (Variables|Settings|GetHealthPerTick))))
      (|CallParentFunction))))

(fn ApplyShields
  (bind _requireshealingorshields (CallFunction|CheckforHealingorShield))
  (if _requireshealingorshields
    (bind _returnvalue (Ability|GameplayCue|IsGameplayCueActive (Ability|GetAvatarActorfromActorInfo) (Variables|Default|GetGameplayCue_Shields)))
    (if _returnvalue
      (|CallParentFunction)
      (if (>= (FortPawn|GetShield (CallFunction|GetAvatarasFortPlayerPawn)) (ScalableFloat|GetValueatLevel (Variables|Settings|GetShieldUpperLimit)))
        (CallFunction|RemoveGC false true))
      (else
        (GameplayCue|AddGameplayCueOnActor(Looping) (Ability|GetAvatarActorfromActorInfo) (Variables|Default|GetGameplayCue_Shields) (Ability|GameplayCue|MakeGameplayCueParameters 0.0 (ScalableFloat|GetValueatLevel (Variables|Settings|GetShieldPerTick))))
        (|CallParentFunction)))
    (else
      (CallFunction|RemoveGC false true))))

(fn CheckForHealingOrShield
  (bind _outfortplayerpawn (CallFunction|GetAvatarasValidFortPlayerPawn)
    (:ValidFortPlayerPawn
      (return (or (< (FortPawn|GetHealth _outfortplayerpawn) (ScalableFloat|GetValueatLevel (Variables|Settings|GetHealthUpperLimit))) (< (FortPawn|GetShield _outfortplayerpawn) (ScalableFloat|GetValueatLevel (Variables|Settings|GetShieldUpperLimit))))))))

(fn RemoveGC
  (params (RemoveHealth? :type Boolean) (RemoveShield? :type Boolean) (RemoveBoth? :type Boolean))
  (bind _returnvalue (Ability|GetAvatarActorfromActorInfo))
  (bind _shieldpertick (Variables|Settings|GetShieldPerTick))
  (bind _returnvalue_2 (ScalableFloat|GetValueatLevel _shieldpertick))
  (bind _healthpertick (Variables|Settings|GetHealthPerTick))
  (bind _returnvalue_3 (ScalableFloat|GetValueatLevel _healthpertick))
  (bind _returnvalue_4 (Ability|GameplayCue|MakeGameplayCueParameters 0.0 _returnvalue_3))
  (bind _returnvalue_5 (Ability|GameplayCue|MakeGameplayCueParameters 0.0 _returnvalue_2))
  (if RemoveBoth?
    (GameplayCue|RemoveGameplayCueOnActor(Looping) _returnvalue (Variables|Default|GetGameplayCue_Shields) _returnvalue_5)
    (GameplayCue|RemoveGameplayCueOnActor(Looping) _returnvalue (Variables|Default|GetGameplayCue_Health) _returnvalue_4)
    (elif RemoveHealth?
      (GameplayCue|RemoveGameplayCueOnActor(Looping) _returnvalue (Variables|Default|GetGameplayCue_Health) _returnvalue_4)
      (elif RemoveShield?
        (GameplayCue|RemoveGameplayCueOnActor(Looping) _returnvalue (Variables|Default|GetGameplayCue_Shields) _returnvalue_5)))))
