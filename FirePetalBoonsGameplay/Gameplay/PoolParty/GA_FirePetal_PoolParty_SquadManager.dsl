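; Pool Party boon gameplay ability that shares the regen ability with nearby squadmates while its owner swims.
; Water tag waits drive a repeating cycle, with a delayed recheck after water jumps before the effect ends.
; If WorksOnSquadMates is enabled, a fast timer grants GA_FirePetal_PoolParty_Regen within Distance and clears it beyond, tracking handles per player state.
; Leaving water clears all grants and the timer; the end-ability event is empty.
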(event Ability|EventOnEndAbility
  (params (bWasCancelled :type Boolean)))

(event Ability|EventActivateAbility
  ; Refactor this code to a shared space between this and GA_FirePetal_PoolParty_Regen
  (CallFunction|WaitforSwimming))

(event Custom|WaitForSwimming
  ; Refactor this code to a shared space between this and GA_FirePetal_PoolParty_Regen
  ;;[1 0.2718 0.06097] Activate when swimming
  (Ability|Tasks|WaitGameplayTagAdd (Variables|Default|GetWaterTag) 0 true
    (:Added
      (CallFunction|WaitforStopSwimming)
      (if (> (ScalableFloat|GetValueasInteger (Variables|Default|GetWorksOnSquadMates)) 0)
        (CallFunction|AddSqaudMatesRegenAbility)
        (bind _returnvalue (Utilities|Time|SetTimerbyEvent (delegate Custom|Manage) 0.25 true))))))

(event Custom|WaitForStopSwimming
  ; Refactor this code to a shared space between this and GA_FirePetal_PoolParty_Regen
  ;;[1 0 0.933] Wait for the player to begin swimming
  (Ability|Tasks|WaitGameplayTagRemove (Variables|Default|GetWaterTag) 0 true
    (:Removed
      (bind _outfortplayerpawn (CallFunction|GetAvatarasValidFortPlayerPawn)
        (:ValidFortPlayerPawn
          (if (Variables|/Script/FortniteGame.FortPlayerPawn|GetbIsWaterJump _outfortplayerpawn)
            (Utilities|Time|SetTimerbyEvent (EventDispatchers|CreateEvent :!event Custom|CheckForInWater) 1.0)
            (else
              (CallFunction|EndEffect))))))))

(event Custom|Manage
  ; Refactor this code to a shared space between this and GA_FirePetal_PoolParty_Regen
  ;; Manage Squadmates
  (CallFunction|RemoveSquadMatesRegenAbility)
  (CallFunction|AddSqaudMatesRegenAbility))

(event Custom|CheckForInWater
  ; Refactor this code to a shared space between this and GA_FirePetal_PoolParty_Regen
  ;;[0 0.2854 1] Check if the player is back in the water after jumping out
  (bind _outfortplayerpawn (CallFunction|GetAvatarasValidFortPlayerPawn)
    (:ValidFortPlayerPawn
      (if (GameplayTags|HasMatchingGameplayTag _outfortplayerpawn (Variables|Default|GetWaterTag))
        (CallFunction|WaitforStopSwimming)
        (else
          (CallFunction|EndEffect))))))

(event Custom|EndEffect
  (bind _squadmatesgivenability (Variables|Default|GetSquadMatesGivenAbility))
  ; Refactor this code to a shared space between this and GA_FirePetal_PoolParty_Regen
  ;;[1 0.5539 0] Setup the wait again when we leave water
  (CallFunction|WaitforSwimming)
  (bind _keys (Utilities|Map|Keys _squadmatesgivenability))
  (for _array_element _keys
    (GameplayAbilities|ClearAbility (Variables|/Script/FortniteGame.FortPlayerState|GetAbilitySystemComponent _array_element) (Utilities|Map|Find _squadmatesgivenability _array_element)))
  (Utilities|Time|ClearandInvalidateTimerbyHandle (stale Custom|WaitForSwimming _returnvalue))
  (Utilities|Map|Clear (Variables|Default|GetSquadMatesGivenAbility)))

(fn AddSqaudMatesRegenAbility
  (bind (_value _returnvalue) (Utilities|Map|Find (Variables|Default|GetSquadMatesGivenAbility) (Variables|Default|GetCurrentSquadMate_Add)))
  (bind _asfort_player_state_athena (Utilities|Casting|CastToFortPlayerStateAthena (Variables|/Script/Engine.Pawn|GetPlayerState (CallFunction|GetAvatarasFortPlayerPawn))))
  (bind _returnvalue_2 (Class|FortGameStateAthena|GetSquadMemberPlayerStates (Fortnite|Athena|GameState|GetGameStateAthena) _asfort_player_state_athena false))
  (for _array_element _returnvalue_2
    (Variables|Default|SetCurrentSquadMate_Add _array_element)
    (if _returnvalue
      _
      (elif (> (Math|Vector|Distance(Vector) (Transformation|GetActorLocation (PlayerState|GetCurrentPawn (Variables|Default|GetCurrentSquadMate_Add))) (Transformation|GetActorLocation (CallFunction|GetAvatarasFortPlayerPawn))) (Variables|Default|GetDistance))
        _
        (else
          (bind _returnvalue_3 (GameplayAbilities|GiveAbility (Variables|/Script/FortniteGame.FortPlayerState|GetAbilitySystemComponent (Variables|Default|GetCurrentSquadMate_Add)) "/FirePetalBoonsGameplay/Gameplay/PoolParty/GA_FirePetal_PoolParty_Regen.GA_FirePetal_PoolParty_Regen_C" 0 -1))
          (Utilities|Map|Add (Variables|Default|GetSquadMatesGivenAbility) (Variables|Default|GetCurrentSquadMate_Add) _returnvalue_3))))))

(fn RemoveSquadMatesRegenAbility
  ; Remove any that left range
  (bind _keys (Utilities|Map|Keys (Variables|Default|GetSquadMatesGivenAbility)))
  (for _array_element _keys
    (Variables|Default|SetCurrentSquadMate_Remove _array_element)
    (if (> (Math|Vector|Distance(Vector) (Transformation|GetActorLocation (PlayerState|GetCurrentPawn (Variables|Default|GetCurrentSquadMate_Remove))) (Transformation|GetActorLocation (CallFunction|GetAvatarasFortPlayerPawn))) (Variables|Default|GetDistance))
      (GameplayAbilities|ClearAbility (Variables|/Script/FortniteGame.FortPlayerState|GetAbilitySystemComponent (Variables|Default|GetCurrentSquadMate_Remove)) (Utilities|Map|Find (Variables|Default|GetSquadMatesGivenAbility) (Variables|Default|GetCurrentSquadMate_Remove)))
      (Utilities|Map|Remove (Variables|Default|GetSquadMatesGivenAbility) (Variables|Default|GetCurrentSquadMate_Remove)))))
